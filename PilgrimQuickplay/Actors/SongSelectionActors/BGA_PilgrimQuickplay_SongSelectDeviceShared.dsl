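; Shared song-selection device actor for the Pilgrim Quickplay green room that opens the music library on interaction or routed event requests.
; It asynchronously creates a curated or standard menu for the primary local player and broadcasts usage.
; While open, it silences ambience, disables world rendering and, on non-touch UI, removes backstage HUD scenes, restoring them on close.
; Green-room timer events can close the menu and lock interaction, which is also blocked for touch input.
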(event EventBeginPlay
  (bind _menuwidget (Variables|Default|GetMenuWidget))
  (Class|FortAsyncActionCurrentPlaylistReady|PlaylistDataReadyAsync
    (:OnReady
      (Networking|CanExecuteCosmeticEvents
        (:True
          (bind _eventrouter (Pilgrim|PilgrimWaitForEventRouter (Class|SparksMusicPlayspace|GetSparksMusicPlayspaceOfClass "/PilgrimQuickplay/Playspaces/BP_PilgrimQuickplayPlayspace.BP_PilgrimQuickplayPlayspace_C"))
            (:OnEventRouterAvailable
              (bind _output_get (Variables|Default|SetEvent%20Router _eventrouter))
              ;[1 0.5011 0] Listen to timer events - close the song selection menu if it is open when the game is transitioning to the play state - don't allow the screen to be opened again during the transition time
              (bind (_asynctaskproxy _payload) (GameplayEventRouter|StartListeningToEvent _output_get "/PilgrimQuickplay/Items/PilgrimEvent_GreenRoomTimerStateChanged.PilgrimEvent_GreenRoomTimerStateChanged")
                (:then
                  ;[0.7396 0.6655 0.1115 0.5] Respond to the song menu being closed
                  (GameplayEventRouter|StartListeningToEvent (Variables|Default|GetEvent%20Router) "/PilgrimQuickplay/Actors/GreenRoomPlayerDisplay/PilgrimEvent_Playlist_Display_Closed.PilgrimEvent_Playlist_Display_Closed"
                    (:then
                      (Utilities|FlowControl|Sequence
                        (:then_0
                          ;[0.7396 0.6655 0.1115 0.5] Respond to external requests to OPEN the song menu
                          (bind (_asynctaskproxy_2 _payload_2) (GameplayEventRouter|StartListeningToEvent (Variables|Default|GetEvent%20Router) "/PilgrimQuickplay/Actors/GreenRoomPlayerDisplay/PilgrimEvent_Playlist_Display_RequestOpen.PilgrimEvent_Playlist_Display_RequestOpen")
                            (:OnEventReceived
                              (CallFunction|OpenSongPickerMenuforPlayer (Utilities|Struct|BreakPilgrimEventPlaylistDisplayRequestOpen _payload_2)))))
                        (:then_1
                          (bind (_asynctaskproxy_3 _payload_3) (GameplayEventRouter|StartListeningToEvent (Variables|Default|GetEvent%20Router) "/Script/PilgrimQuickplayRuntime.PilgrimQuickplayGreenRoomPlaylistDisplayRequest")
                            (:OnEventReceived
                              (CallFunction|OpenSongPickerMenuforPlayer (Utilities|Struct|BreakPilgrimQuickplayGreenRoomPlaylistDisplayRequest _payload_3)))))
                        (:then_2
                          ;[0.7396 0.6655 0.1115 0.5] Respond to external requests to CLOSE the song menu
                          (bind (_asynctaskproxy_4 _payload_4) (GameplayEventRouter|StartListeningToEvent (Variables|Default|GetEvent%20Router) "/Script/PilgrimQuickplayRuntime.PilgrimQuickplayGreenRoomPlaylistHideRequest")
                            (:OnEventReceived
                              (CallFunction|CloseSongPickerMenuForPlayer (Utilities|Struct|BreakPilgrimQuickplayGreenRoomPlaylistHideRequest _payload_4)))))))
                    (:OnEventReceived
                      (CallFunction|OnMenuClosed))))
                (:OnEventReceived
                  (bind _payload_greenroomtimerstate_2_04fc39c74228c50f7ae0f7b72d379be4 (Utilities|Struct|BreakPilgrimEventGreenRoomTimerStateChanged _payload))
                  (switch Utilities|FlowControl|Switch|SwitchonGreenRoomTimerState _payload_greenroomtimerstate_2_04fc39c74228c50f7ae0f7b72d379be4
                    (:NewEnumerator0
                      _)
                    (:NewEnumerator1
                      _)
                    (:NewEnumerator2
                      (Variables|Default|SetisLockedByTimer true)
                      (Utilities|IsValid _menuwidget
                        (:Is%20Valid
                          (ActivatableWidget|DeactivateWidget _menuwidget))))
                    (:NewEnumerator3
                      _))
                  (Variables|Default|SetisLockedByTimer false))))))))))

(event Custom|OpenSongPickerMenu
  (params (FortPawn :type Fort%20Pawn%20Object%20Reference))
  (CallFunction|OpenSongPickerMenuforPlayer (Fortnite|GetFortPlayerControllerfromActor FortPawn)))

(event Custom|OpenSongPickerMenuForPlayer
  (params (PlayerController :type Fort%20Player%20Controller%20Object%20Reference))
  (bind _returnvalue (Fortnite|Athena|GameState|GetGameStateAthena))
  (bind _returnvalue_2 (Actor|GetComponentbyClass _returnvalue "/Script/PilgrimQuickplayRuntime.PilgrimSetlistComponent"))
  ;[0.2899 0.9927 1 0.577] Only open the menu for a primary local player controller
  (if (CallFunction|IsPrimaryLocalPlayer PlayerController)
    (if true
      (CallFunction|NotifySongDeviceOpened PlayerController)
      ;[0.4983 1 0.9602] Create and push the song selection widget to the screen
      (Utilities|IsValid _returnvalue_2
        (:Is%20Valid
          (bind _userwidget (Class|FortAsyncActionCreateWidgetAsync|CreateWidgetAsync (select (Class|PilgrimSetlistComponent|IsCuratedSetlistMode _returnvalue_2) (Variables|Default|GetCuratedSetlistMenu) (Variables|Default|GetStandardSongMenu)) PlayerController)
            (:OnComplete
              ;;[1 0.9641 0.1476] On non-mobile interfaces, remove the backstage hud to fix a focus issue
              (if (not (Class|PilgrimCoreBPFL|ShouldUseTouchUIForLocalPlayer))
                (DynamicUI|RemoveScene (WorldSubsystems|GetDynamicUIManager) "/PilgrimQuickplay/Widgets/Scenes/UIScene_PilgrimQuickplay_GreenRoom_Default.UIScene_PilgrimQuickplay_GreenRoom_Default" PlayerController)
                ;[0.15 0.06545 0.04095 0.5] Becasue we are removing Greenroom, in the case that UPP is enabled, we remove it as well so we can add it when the Music Library closes
                (if (Class|PilgrimCoreBPFL|IsUnifiedPartPickerEnabled)
                  (DynamicUI|RemoveScene (WorldSubsystems|GetDynamicUIManager) "/PilgrimQuickplay/Widgets/Scenes/UIScene_PilgrimQuickplay_UPP.UIScene_PilgrimQuickplay_UPP" PlayerController)))
              (FortUIManagement|PushContentWidgetforPlayer PlayerController _userwidget)
              (bind _ascommon_activatable_widget (Utilities|Casting|CastToCommonActivatableWidget _userwidget))
              (Variables|Default|SetMenuWidget _ascommon_activatable_widget)
              (Audio|Components|Audio|SetVolumeMultiplier (Variables|Default|Getorgan_ambience_4_Cue) 0.0 :!class /Script/Engine.AudioComponent)
              (Rendering|SetEnableWorldRendering false) ; Disable world rendering for perf while this menu is up (FORT-731804)
              ;[0 0.4979 0.5365] play a sound when the menu opens. The first time you get a big organ noise, after that it's subtler.
              (Utilities|DoOnceThenDo false
                (:DoOnce
                  (Audio|PlaySound2D "/PilgrimQuickplay/Audio/new_organ.new_organ"))
                (:ThenDo
                  (Audio|PlaySound2D "/PilgrimQuickplay/Audio/new_organ_button.new_organ_button")))
              ;[0.625 1 1] Bind to the OnMenuClosed event - This is the old way. Screens should be updated to broadcast PilgrimEvent_Playlist_Display_Closed instead of using the event dispatcher
              (bind _aswbp_pilgrim_quickplay_song_select_menu_2 (Utilities|Casting|CastToWBP_PilgrimQuickplay_SongSelectMenu2 (Variables|Default|GetMenuWidget)))
              (bind _output_get (Variables|Default|SetSongMenuWidget _aswbp_pilgrim_quickplay_song_select_menu_2))
              (Default|BindEventtoOnMenuClosed _output_get (EventDispatchers|CreateEvent :!event Custom|OnMenuClosed)))))
        (:Is%20Not%20Valid
          (Development|PrintString "PilgrimSetlistComponent Not found!" true true "(R=0.000000,G=0.660000,B=1.000000,A=1.000000)" 2e+01))))))

(event Custom|Close%20Song%20Picker%20Menu%20for%20Player
  (params (Player%20Controller :type Fort%20Player%20Controller%20Object%20Reference))
  (Utilities|IsValid Player%20Controller
    (:Is%20Valid
      (FortUIManagement|PopContentWidgetforPlayer Player%20Controller))))

(event Custom|OnMenuClosed
  (bind _menuwidget (Variables|Default|GetMenuWidget))
  (bind _returnvalue (Widget|GetOwningPlayer _menuwidget))
  (Audio|Components|Audio|SetVolumeMultiplier (Variables|Default|Getorgan_ambience_4_Cue) 0.75 :!class /Script/Engine.AudioComponent)
  (Rendering|SetEnableWorldRendering true) ; Re-enable world rendering when the menu is closed (FORT-731804)
  (Default|UnbindEventfromOnMenuClosed (Variables|Default|GetSongMenuWidget) (delegate Custom|OnMenuClosed))
  ;[0.8337 1 0.2027] For non-mobile devices, re-add the backstage hud
  (if (not (Class|PilgrimCoreBPFL|ShouldUseTouchUIForLocalPlayer))
    (DynamicUI|AddScene (WorldSubsystems|GetDynamicUIManager) "/PilgrimQuickplay/Widgets/Scenes/UIScene_PilgrimQuickplay_GreenRoom_Default.UIScene_PilgrimQuickplay_GreenRoom_Default" _returnvalue)
    ;[0.15 0.06545 0.04095 0.5] Re-adding UPP scene if it is the case
    (if (Class|PilgrimCoreBPFL|IsUnifiedPartPickerEnabled)
      (DynamicUI|AddScene (WorldSubsystems|GetDynamicUIManager) "/PilgrimQuickplay/Widgets/Scenes/UIScene_PilgrimQuickplay_UPP.UIScene_PilgrimQuickplay_UPP" _returnvalue))))

(event Custom|NotifySongDeviceOpened
  (params (Controller :type Controller%20Object%20Reference))
  ;[0.06109 0.9115 0.6916 0.5] Notify listeners that we used the song device
  (bind _asfort_player_state (Utilities|Casting|CastToFortPlayerState (Variables|/Script/Engine.Controller|GetPlayerState Controller)))
  (GameplayEventRouter|BroadcastEvent (Variables|Default|GetEvent%20Router) (Utilities|Struct|MakePilgrimEventPlayerExperiencePlayerOpenedSongSelectMenu _asfort_player_state) 0))

(fn BlueprintCanInteract
  (params (InteractingPawn :type Fort%20Pawn%20Object%20Reference) (InteractionBeingAttempted :type EInteractionBeingAttempted%20Enum :default "FirstInteraction") (InteractionType :type TInteractionType%20Enum :default "IT_NoInteraction"))
  (return (not (or (Variables|Default|GetisLockedByTimer) (CallFunction|IsInputMethodActive "Touch")))))

(fn BlueprintGetInteractionString
  (params (InteractingPawn :type Fort%20Pawn%20Object%20Reference) (InteractionBeingAttempted :type EInteractionBeingAttempted%20Enum :default "FirstInteraction"))
  (return (Utilities|String|SelectText "" (Variables|Default|GetOpen%20Music%20Library%20Interaction%20String) (CallFunction|IsInputMethodActive "Touch"))))

(fn BlueprintOnLocalInteract
  (params (InteractingPawn :type Fort%20Player%20Pawn%20Object%20Reference) (InteractionBeingAttempted :type EInteractionBeingAttempted%20Enum :default "FirstInteraction"))
  (CallFunction|OpenSongPickerMenu InteractingPawn)
  (return true))

(fn CanDisplayEasyInteractButton
  (return false))

(fn IsInputMethodActive
  (params (Input%20Type :type ECommonInputType%20Enum :default "MouseAndKeyboard"))
  (bind _returnvalue (Fortnite|GetFirstLocalFortPlayerController true))
  (bind _returnvalue_2 (PlayerController|LocalPlayerSubsystems|GetCommonInputSubsystem _returnvalue))
  (bind _output_get (LocalVariables|SetFirstPlayerController _returnvalue))
  (Utilities|IsValid _output_get
    (:Is%20Valid
      (Utilities|IsValid _returnvalue_2
        (:Is%20Valid
          (return (CommonInputSubsystem|IsInputMethodActive _returnvalue_2 (Variables|GetInput%20Type))))
        (:Is%20Not%20Valid
          _)))
    (:Is%20Not%20Valid
      _))
  (return false))

(fn IsPrimaryLocalPlayer
  (params (Player%20Controller :type Fort%20Player%20Controller%20Object%20Reference))
  (bind _returnvalue (Actor|GetComponentbyClass Player%20Controller "/Script/PilgrimCoreRuntime.PilgrimPlayerControllerComponent"))
  (Utilities|IsValid Player%20Controller
    (:Is%20Valid
      (Utilities|IsValid _returnvalue
        (:Is%20Valid
          (return (and (Class|PilgrimPlayerControllerComponent|IsPrimaryLocalPlayer _returnvalue) (Pawn|IsLocalPlayerController Player%20Controller))))
        (:Is%20Not%20Valid
          _)))
    (:Is%20Not%20Valid
      _))
  (return false))

(fn ConstructionScript
  (bind _interactionbounds (Variables|Default|GetInteractionBounds))
  ;[0.625 1 1] Adjust the size and position of the interaction bounds
  (Components|Box|SetBoxExtent _interactionbounds (Variables|Default|GetInteractionBoundsDimensions) true)
  (Transformation|SetRelativeLocation _interactionbounds (Variables|Default|GetInteractionBoundsLocation) false false))
