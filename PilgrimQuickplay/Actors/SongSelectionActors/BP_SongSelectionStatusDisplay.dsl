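; Song selection status actor for Pilgrim Quickplay that shows the custom setlist, with intensities for the local player's instrument, on a widget component.
; After playlist data and the event router are ready, setlist and local instrument change events update the cached widget.
; It also reads the playspace's current setlist to initialise late joiners.
; Unless set always visible, pawn overlaps with a range-scaled collision volume toggle widget visibility cosmetically.
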(event EventTick
  (params (DeltaSeconds :type Float%20%28single-precision%29)))

(event EventBeginPlay
  (bind _returnvalue (Class|SparksMusicPlayspace|GetSparksMusicPlayspaceOfClass "/PilgrimQuickplay/Playspaces/BP_PilgrimQuickplayPlayspace.BP_PilgrimQuickplayPlayspace_C"))
  (bind _localinstrument (Variables|Default|GetLocalInstrument))
  (bind _setlistdisplayalwaysvisible (Variables|Default|GetSetlistDisplayAlwaysVisible))
  (Utilities|FlowControl|Sequence
    (:then_0
      ;[0.6237 0.9701 1] Set the visibility of the displays, then if we aren't always visible, set the scale of the visibility trigger volume
      (CallFunction|SetDisplayVisibility _setlistdisplayalwaysvisible)
      (if (not _setlistdisplayalwaysvisible)
        ;[0.6769 0.9901 1] Set the scale of the visibility interaction collision
        (Transformation|SetRelativeScale3D (Variables|Default|GetVisibilityCollision) (Math|Conversions|ToVector(Float) (Variables|Default|GetVisibilityInteractionRange)))))
    (:then_1
      ;[0.5121 1 0.869] Cache the setlist widget
      (bind _aswbp_pilgrim_quickplay_setlist_display_custom (Utilities|Casting|CastToWBP_PilgrimQuickplay_SetlistDisplayCustom (UserInterface|GetWidget (Variables|Default|GetCurrentSetlistWidget))))
      (Variables|Default|SetCustomSetlistWidget _aswbp_pilgrim_quickplay_setlist_display_custom))
    (:then_2
      ;[0.562 0.9801 1] Get the event router, and start listening for when the setlist changes
      (Class|FortAsyncActionCurrentPlaylistReady|PlaylistDataReadyAsync
        (:OnReady
          (bind _eventrouter (Pilgrim|PilgrimWaitForEventRouter 0)
            (:OnEventRouterAvailable
              (bind _output_get (Variables|Default|SetEventRouter _eventrouter))
              (Utilities|FlowControl|Sequence
                (:then_0
                  ;[0.5121 1 0.869] When the setlist changes, update the display
                  (bind (_asynctaskproxy _payload) (GameplayEventRouter|StartListeningToEvent _output_get "/Script/PilgrimCoreRuntime.PilgrimEvent_BandState_CustomSetlistChanged")
                    (:OnEventReceived
                      (if (Utilities|Enum|Equal(Enum) _localinstrument "TrackNone")
                        (CallFunction|GetLocalInstrument))
                      (Class|WBPPilgrimQuickplaySetlistDisplayCustom|UpdateCustomSetlist (Variables|Default|GetCustomSetlistWidget) (Class|PilgrimSetlistInterface|GetSongList (Utilities|Struct|BreakPilgrimEventBandStateCustomSetlistChanged _payload)) _localinstrument))))
                (:then_1
                  ;[0.5198 1 0.9701] Listen for instrument changes to update the displayed intensity
                  (bind (_asynctaskproxy_2 _payload_2) (GameplayEventRouter|StartListeningToEvent _output_get "/Script/PilgrimCoreRuntime.PilgrimEvent_PlayerState_InstrumentChanged")
                    (:OnEventReceived
                      ;;[0.4379 1 0.3896] Only do this for local instrument changes
                      (bind (_payload_playerstate _payload_newinstrumentchoice) (Utilities|Struct|BreakPilgrimEventPlayerStateInstrumentChanged _payload_2))
                      (Utilities|IsValid (PlayerState|GetPlayerController _payload_playerstate)
                        (:Is%20Valid
                          (if (Pawn|IsLocalController (PlayerState|GetPlayerController _payload_playerstate))
                            (Variables|Default|SetLocalInstrument _payload_newinstrumentchoice)
                            (Class|WBPPilgrimQuickplaySetlistDisplayCustom|UpdateAllIntensitiesonInstrumentChanged (Variables|Default|GetCustomSetlistWidget) _payload_newinstrumentchoice)))))))
                (:then_2
                  ;[0.6123 1 0.9601] Get the custom setlist from the playspace
                  (Utilities|IsValid _returnvalue
                    (:Is%20Valid
                      (bind _output_get_2 (Variables|Default|SetCustomSetlist (Variables|/PilgrimQuickplay/Playspaces/BP_PilgrimQuickplayPlayspace.BP_PilgrimQuickplayPlayspace_C|GetPilgrimCustomSetlist _returnvalue)))
                      (Utilities|IsCustomSetlistValid
                        (:Valid
                          ;[0.5828 1 0.9801] Setup the display for late joiners (They shouldn't have an instrument set yet, so defaulting to guitar, or if guitar is set immediately, should update correctly)
                          (bind _aswbp_pilgrim_quickplay_setlist_display_custom_2 (Utilities|Casting|CastToWBP_PilgrimQuickplay_SetlistDisplayCustom (UserInterface|GetWidget (Variables|Default|GetCurrentSetlistWidget))))
                          (bind _localinstrument_2 (CallFunction|GetLocalInstrument))
                          (Class|WBPPilgrimQuickplaySetlistDisplayCustom|UpdateCustomSetlist _aswbp_pilgrim_quickplay_setlist_display_custom_2 (Class|PilgrimSetlistInterface|GetSongList _output_get_2) _localinstrument_2))))
                    (:Is%20Not%20Valid
                      (Development|PrintString "No valid PilgrimQuickplay Playspace found"))))))))))))

(event Custom|SetDisplayVisibility
  (params (DisplayIsVisible :type Boolean))
  ;[0.5624 0.9901 1] When the player enters or leaves the visibility collision, show or hide the displays with some nice animation
  (Networking|CanExecuteCosmeticEvents
    (:True
      (if (not (Variables|Default|GetSetlistDisplayAlwaysVisible))
        (Rendering|SetVisibility (Variables|Default|GetCurrentSetlistWidget) DisplayIsVisible false)))))

(event Collision|EventActorBeginOverlap
  (params (OtherActor :type Actor%20Object%20Reference))
  ;[0.5624 0.9901 1] When the player enters or leaves the visibility collision, show or hide the displays with some nice animation
  (Utilities|Casting|CastToFortPawn OtherActor)
  (CallFunction|SetDisplayVisibility true))

(event Collision|EventActorEndOverlap
  (params (OtherActor :type Actor%20Object%20Reference))
  ;[0.5624 0.9901 1] When the player enters or leaves the visibility collision, show or hide the displays with some nice animation
  (Utilities|Casting|CastToFortPawn OtherActor)
  (CallFunction|SetDisplayVisibility false))

(fn GetLocalInstrument
  (bind _array_element (Utilities|Array|ForEachLoopwithBreak (Class|SparksMusicPlayspace|GetAllPlayerStates (Class|SparksMusicPlayspace|GetSparksMusicPlayspaceOfClass "/PilgrimQuickplay/Playspaces/BP_PilgrimQuickplayPlayspace.BP_PilgrimQuickplayPlayspace_C")))
    (:LoopBody
      (Utilities|IsValid (PlayerState|GetPlayerController _array_element)
        (:Is%20Valid
          (if (Pawn|IsLocalPlayerController (PlayerState|GetPlayerController _array_element))
            (bind _returnvalue (Class|PilgrimQuickplayPlayerStateComponent|GetPilgrimQuickplayPlayerStateComponent _array_element))
            (Variables|Default|SetLocalInstrument (Class|PilgrimCorePlayerStateComponent|GetInstrumentChoice _returnvalue))))))
    (:Completed
      _))
  (return (Variables|Default|GetLocalInstrument)))

(fn ConstructionScript)
